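# Sound cue for the DustyWing cosmetic companion's flip foley, classed as a pet creature sound.
# A quality-level switch feeds one of two equally weighted random pools: six flip waves for the first two levels, a four-wave subset for the last two.
# It uses companion foley attenuation and borrows the shared quadruped footstep concurrency rather than a flip-specific one.
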SoundCue /CosmeticCompanions/Assets/Other/DustyWing/Audio/Foley/SC_Mimosa_Other_DustyWing_Flip
root=QualityLevel_0
nodes [
 {id=QualityLevel_0, children [Random_0,Random_0,Random_1,Random_1]}
 {
  id=Random_0
  children [WavePlayer_0, WavePlayer_1, WavePlayer_10, WavePlayer_11, WavePlayer_12, WavePlayer_13]
  properties {weights [1,1,1,1,1,1]}
 }
 {id=Random_1, children [WavePlayer_14, WavePlayer_15, WavePlayer_16, WavePlayer_17], properties {weights [1,1,1,1]}}
 {id=WavePlayer_0, properties {soundWaveAssetPtr=/CosmeticCompanions/Assets/Other/DustyWing/Audio/Foley/Wavs/Mimosa_DustyWing_Flip_01}}
 {id=WavePlayer_1, properties {soundWaveAssetPtr=/CosmeticCompanions/Assets/Other/DustyWing/Audio/Foley/Wavs/Mimosa_DustyWing_Flip_02}}
 {id=WavePlayer_10, properties {soundWaveAssetPtr=/CosmeticCompanions/Assets/Other/DustyWing/Audio/Foley/Wavs/Mimosa_DustyWing_Flip_06}}
 {id=WavePlayer_11, properties {soundWaveAssetPtr=/CosmeticCompanions/Assets/Other/DustyWing/Audio/Foley/Wavs/Mimosa_DustyWing_Flip_05}}
 {id=WavePlayer_12, properties {soundWaveAssetPtr=/CosmeticCompanions/Assets/Other/DustyWing/Audio/Foley/Wavs/Mimosa_DustyWing_Flip_04}}
 {id=WavePlayer_13, properties {soundWaveAssetPtr=/CosmeticCompanions/Assets/Other/DustyWing/Audio/Foley/Wavs/Mimosa_DustyWing_Flip_03}}
 {id=WavePlayer_14, properties {soundWaveAssetPtr=&Mimosa_DustyWing_Flip_06}}
 {id=WavePlayer_15, properties {soundWaveAssetPtr=&Mimosa_DustyWing_Flip_05}}
 {id=WavePlayer_16, properties {soundWaveAssetPtr=&Mimosa_DustyWing_Flip_04}}
 {id=WavePlayer_17, properties {soundWaveAssetPtr=&Mimosa_DustyWing_Flip_03}}
]
properties {
 attenuationSettings=/Game/Sounds/Attenuation/AS_Companion_Foley
 maxDistance=1200
 priority=0
 soundClassObject=&Fort_Creatures_Pets
 volumeMultiplier=2.1
 concurrencySet [/CosmeticCompanions/Assets/Audio_Shared/Foley/Quadraped/CONC_Mimosa_Quadraped_Footstep]
}
