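# Sound cue for the jump foley of the DustyWing cosmetic companion, played in the pets sound class with companion foley attenuation.
# A quality level switch picks a random pool of ten jump waves for the first two levels, or a reduced pool of the first four for the last two.
# A modulator adds slight volume variation at fixed pitch, and the cue shares the quadruped footstep concurrency.
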SoundCue /CosmeticCompanions/Assets/Other/DustyWing/Audio/Foley/SC_Mimosa_Other_DustyWing_Jump
root=Modulator_0
nodes [
 {id=Modulator_0, children [QualityLevel_0], properties {pitchMax=1, pitchMin=1, volumeMin=0.9}}
 {id=QualityLevel_0, children [Random_2,Random_2,Random_3,Random_3]}
 {
  id=Random_2
  children [WavePlayer_1, WavePlayer_2, WavePlayer_3, WavePlayer_4, WavePlayer_5, WavePlayer_6, WavePlayer_7, WavePlayer_8, WavePlayer_9, WavePlayer_10]
  properties {weights [1,1,1,1,1,1,1,1,1,1]}
 }
 {id=Random_3, children [WavePlayer_11, WavePlayer_12, WavePlayer_13, WavePlayer_14], properties {weights [1,1,1,1]}}
 {id=WavePlayer_1, properties {soundWaveAssetPtr=/CosmeticCompanions/Assets/Other/DustyWing/Audio/Foley/Wavs/Mimosa_DustyWing_Jump_01}}
 {id=WavePlayer_10, properties {soundWaveAssetPtr=/CosmeticCompanions/Assets/Other/DustyWing/Audio/Foley/Wavs/Mimosa_DustyWing_Jump_10}}
 {id=WavePlayer_11, properties {soundWaveAssetPtr=&Mimosa_DustyWing_Jump_01}}
 {id=WavePlayer_12, properties {soundWaveAssetPtr=/CosmeticCompanions/Assets/Other/DustyWing/Audio/Foley/Wavs/Mimosa_DustyWing_Jump_02}}
 {id=WavePlayer_13, properties {soundWaveAssetPtr=/CosmeticCompanions/Assets/Other/DustyWing/Audio/Foley/Wavs/Mimosa_DustyWing_Jump_03}}
 {id=WavePlayer_14, properties {soundWaveAssetPtr=/CosmeticCompanions/Assets/Other/DustyWing/Audio/Foley/Wavs/Mimosa_DustyWing_Jump_04}}
 {id=WavePlayer_2, properties {soundWaveAssetPtr=&Mimosa_DustyWing_Jump_02}}
 {id=WavePlayer_3, properties {soundWaveAssetPtr=&Mimosa_DustyWing_Jump_03}}
 {id=WavePlayer_4, properties {soundWaveAssetPtr=&Mimosa_DustyWing_Jump_04}}
 {id=WavePlayer_5, properties {soundWaveAssetPtr=/CosmeticCompanions/Assets/Other/DustyWing/Audio/Foley/Wavs/Mimosa_DustyWing_Jump_05}}
 {id=WavePlayer_6, properties {soundWaveAssetPtr=/CosmeticCompanions/Assets/Other/DustyWing/Audio/Foley/Wavs/Mimosa_DustyWing_Jump_06}}
 {id=WavePlayer_7, properties {soundWaveAssetPtr=/CosmeticCompanions/Assets/Other/DustyWing/Audio/Foley/Wavs/Mimosa_DustyWing_Jump_07}}
 {id=WavePlayer_8, properties {soundWaveAssetPtr=/CosmeticCompanions/Assets/Other/DustyWing/Audio/Foley/Wavs/Mimosa_DustyWing_Jump_08}}
 {id=WavePlayer_9, properties {soundWaveAssetPtr=/CosmeticCompanions/Assets/Other/DustyWing/Audio/Foley/Wavs/Mimosa_DustyWing_Jump_09}}
]
properties {
 attenuationSettings=/Game/Sounds/Attenuation/AS_Companion_Foley
 maxDistance=1200
 priority=0
 soundClassObject=&Fort_Creatures_Pets
 volumeMultiplier=1.3
 concurrencySet [/CosmeticCompanions/Assets/Audio_Shared/Foley/Quadraped/CONC_Mimosa_Quadraped_Footstep]
}
