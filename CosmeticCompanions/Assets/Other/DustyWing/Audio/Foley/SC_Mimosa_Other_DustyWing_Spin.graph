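# Sound cue that plays the spin foley for the DustyWing cosmetic companion, routed to the pets sound class with companion foley attenuation.
# A quality-level root sends its first two levels to an equal-weight random pick of six spin waves.
# The last two levels share a smaller random pick that reuses only the first four waves.
# Playback is limited by the shared quadruped footstep concurrency rather than a spin-specific one.
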SoundCue /CosmeticCompanions/Assets/Other/DustyWing/Audio/Foley/SC_Mimosa_Other_DustyWing_Spin
root=QualityLevel_0
nodes [
 {id=QualityLevel_0, children [Random_0,Random_0,Random_3,Random_3]}
 {id=Random_0, children [WavePlayer_1,WavePlayer_2,WavePlayer_3,WavePlayer_4,WavePlayer_5,WavePlayer_6], properties {weights [1,1,1,1,1,1]}}
 {id=Random_3, children [WavePlayer_7, WavePlayer_8, WavePlayer_9, WavePlayer_10], properties {weights [1,1,1,1]}}
 {id=WavePlayer_1, properties {soundWaveAssetPtr=/CosmeticCompanions/Assets/Other/DustyWing/Audio/Foley/Wavs/Mimosa_DustyWing_Spin_01}}
 {id=WavePlayer_10, properties {soundWaveAssetPtr=/CosmeticCompanions/Assets/Other/DustyWing/Audio/Foley/Wavs/Mimosa_DustyWing_Spin_04}}
 {id=WavePlayer_2, properties {soundWaveAssetPtr=/CosmeticCompanions/Assets/Other/DustyWing/Audio/Foley/Wavs/Mimosa_DustyWing_Spin_02}}
 {id=WavePlayer_3, properties {soundWaveAssetPtr=/CosmeticCompanions/Assets/Other/DustyWing/Audio/Foley/Wavs/Mimosa_DustyWing_Spin_03}}
 {id=WavePlayer_4, properties {soundWaveAssetPtr=&Mimosa_DustyWing_Spin_04}}
 {id=WavePlayer_5, properties {soundWaveAssetPtr=/CosmeticCompanions/Assets/Other/DustyWing/Audio/Foley/Wavs/Mimosa_DustyWing_Spin_05}}
 {id=WavePlayer_6, properties {soundWaveAssetPtr=/CosmeticCompanions/Assets/Other/DustyWing/Audio/Foley/Wavs/Mimosa_DustyWing_Spin_06}}
 {id=WavePlayer_7, properties {soundWaveAssetPtr=&Mimosa_DustyWing_Spin_01}}
 {id=WavePlayer_8, properties {soundWaveAssetPtr=&Mimosa_DustyWing_Spin_02}}
 {id=WavePlayer_9, properties {soundWaveAssetPtr=&Mimosa_DustyWing_Spin_03}}
]
properties {
 attenuationSettings=/Game/Sounds/Attenuation/AS_Companion_Foley
 maxDistance=1200
 priority=0
 soundClassObject=&Fort_Creatures_Pets
 volumeMultiplier=1.5
 concurrencySet [/CosmeticCompanions/Assets/Audio_Shared/Foley/Quadraped/CONC_Mimosa_Quadraped_Footstep]
}
