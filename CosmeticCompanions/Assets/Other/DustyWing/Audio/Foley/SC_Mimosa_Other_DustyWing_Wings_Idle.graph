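# Sound cue for the idle wing foley of the DustyWing cosmetic companion, in the pets sound class with companion foley attenuation.
# A quality switch sends its first two levels to an equal-weight random pick of sixteen wing waves and its last two to the first eight.
# A modulator at the root varies pitch and volume.
# Concurrency reuses the shared quadruped footstep set.
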SoundCue /CosmeticCompanions/Assets/Other/DustyWing/Audio/Foley/SC_Mimosa_Other_DustyWing_Wings_Idle
root=Modulator_0
nodes [
 {id=Modulator_0, children [QualityLevel_0], properties {pitchMax=1.2, pitchMin=0.9, volumeMax=1, volumeMin=0.8}}
 {id=QualityLevel_0, children [Random_2,Random_2,Random_3,Random_3]}
 {
  id=Random_2
  children [
   WavePlayer_0
   WavePlayer_1
   WavePlayer_2
   WavePlayer_3
   WavePlayer_4
   WavePlayer_5
   WavePlayer_6
   WavePlayer_7
   WavePlayer_8
   WavePlayer_9
   WavePlayer_10
   WavePlayer_11
   WavePlayer_12
   WavePlayer_13
   WavePlayer_14
   WavePlayer_15
  ]
  properties {weights [1,1,1,1,1,1,1,1,1,1,1,1,1,1,1,1]}
 }
 {
  id=Random_3
  children [WavePlayer_16, WavePlayer_17, WavePlayer_18, WavePlayer_19, WavePlayer_20, WavePlayer_21, WavePlayer_22, WavePlayer_23]
  properties {weights [1,1,1,1,1,1,1,1]}
 }
 {id=WavePlayer_0, properties {soundWaveAssetPtr=/CosmeticCompanions/Assets/Other/DustyWing/Audio/Foley/Wavs/Mimosa_DustyWing_Wings_Idle_16}}
 {id=WavePlayer_1, properties {soundWaveAssetPtr=/CosmeticCompanions/Assets/Other/DustyWing/Audio/Foley/Wavs/Mimosa_DustyWing_Wings_Idle_15}}
 {id=WavePlayer_10, properties {soundWaveAssetPtr=/CosmeticCompanions/Assets/Other/DustyWing/Audio/Foley/Wavs/Mimosa_DustyWing_Wings_Idle_06}}
 {id=WavePlayer_11, properties {soundWaveAssetPtr=/CosmeticCompanions/Assets/Other/DustyWing/Audio/Foley/Wavs/Mimosa_DustyWing_Wings_Idle_05}}
 {id=WavePlayer_12, properties {soundWaveAssetPtr=/CosmeticCompanions/Assets/Other/DustyWing/Audio/Foley/Wavs/Mimosa_DustyWing_Wings_Idle_04}}
 {id=WavePlayer_13, properties {soundWaveAssetPtr=/CosmeticCompanions/Assets/Other/DustyWing/Audio/Foley/Wavs/Mimosa_DustyWing_Wings_Idle_03}}
 {id=WavePlayer_14, properties {soundWaveAssetPtr=/CosmeticCompanions/Assets/Other/DustyWing/Audio/Foley/Wavs/Mimosa_DustyWing_Wings_Idle_02}}
 {id=WavePlayer_15, properties {soundWaveAssetPtr=/CosmeticCompanions/Assets/Other/DustyWing/Audio/Foley/Wavs/Mimosa_DustyWing_Wings_Idle_01}}
 {id=WavePlayer_16, properties {soundWaveAssetPtr=/CosmeticCompanions/Assets/Other/DustyWing/Audio/Foley/Wavs/Mimosa_DustyWing_Wings_Idle_08}}
 {id=WavePlayer_17, properties {soundWaveAssetPtr=/CosmeticCompanions/Assets/Other/DustyWing/Audio/Foley/Wavs/Mimosa_DustyWing_Wings_Idle_07}}
 {id=WavePlayer_18, properties {soundWaveAssetPtr=&Mimosa_DustyWing_Wings_Idle_06}}
 {id=WavePlayer_19, properties {soundWaveAssetPtr=&Mimosa_DustyWing_Wings_Idle_05}}
 {id=WavePlayer_2, properties {soundWaveAssetPtr=/CosmeticCompanions/Assets/Other/DustyWing/Audio/Foley/Wavs/Mimosa_DustyWing_Wings_Idle_14}}
 {id=WavePlayer_20, properties {soundWaveAssetPtr=&Mimosa_DustyWing_Wings_Idle_04}}
 {id=WavePlayer_21, properties {soundWaveAssetPtr=&Mimosa_DustyWing_Wings_Idle_03}}
 {id=WavePlayer_22, properties {soundWaveAssetPtr=&Mimosa_DustyWing_Wings_Idle_02}}
 {id=WavePlayer_23, properties {soundWaveAssetPtr=&Mimosa_DustyWing_Wings_Idle_01}}
 {id=WavePlayer_3, properties {soundWaveAssetPtr=/CosmeticCompanions/Assets/Other/DustyWing/Audio/Foley/Wavs/Mimosa_DustyWing_Wings_Idle_13}}
 {id=WavePlayer_4, properties {soundWaveAssetPtr=/CosmeticCompanions/Assets/Other/DustyWing/Audio/Foley/Wavs/Mimosa_DustyWing_Wings_Idle_12}}
 {id=WavePlayer_5, properties {soundWaveAssetPtr=/CosmeticCompanions/Assets/Other/DustyWing/Audio/Foley/Wavs/Mimosa_DustyWing_Wings_Idle_11}}
 {id=WavePlayer_6, properties {soundWaveAssetPtr=/CosmeticCompanions/Assets/Other/DustyWing/Audio/Foley/Wavs/Mimosa_DustyWing_Wings_Idle_10}}
 {id=WavePlayer_7, properties {soundWaveAssetPtr=/CosmeticCompanions/Assets/Other/DustyWing/Audio/Foley/Wavs/Mimosa_DustyWing_Wings_Idle_09}}
 {id=WavePlayer_8, properties {soundWaveAssetPtr=&Mimosa_DustyWing_Wings_Idle_08}}
 {id=WavePlayer_9, properties {soundWaveAssetPtr=&Mimosa_DustyWing_Wings_Idle_07}}
]
properties {
 attenuationSettings=/Game/Sounds/Attenuation/AS_Companion_Foley
 maxDistance=1200
 pitchMultiplier=0.9
 priority=0
 soundClassObject=&Fort_Creatures_Pets
 volumeMultiplier=0.5
 concurrencySet [/CosmeticCompanions/Assets/Audio_Shared/Foley/Quadraped/CONC_Mimosa_Quadraped_Footstep]
}
